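# Material function for the SunTide player body that adds animated purple energy and glowing runes to incoming material attributes, blended by an Alpha input.
# Scrolling, warped noise projected in pre-skinned space so it follows the mesh is shaped by Fresnel into energy masks.
# A rune mask and tiled height texture, with hashed per-cell randomness and sine pulsing, shape the runes.
# ThinFilmFunction supplies a metal layer, and emissive passes through DeriveHDRfromLDR.
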MaterialFunction /Game/Characters/Player/Male/Medium/Bodies/M_MED_SunTide/Materials/MF_Suntide
graph {nodes [
  n1 = FunctionInput() {bUsePreviewValueAsDefault=true, inputName=Alpha, inputType=FunctionInput_Scalar, previewValue=<1 0 0 1>}
  FunctionOutput(None=BlendMaterialAttributes(A=NamedRerouteUsage(), B=NamedRerouteUsage(), Alpha=@n1))
  n2 = TextureCoordinate()
  n3 = ScalarParameter() {defaultValue=20, group=Suntide, parameterName=RuneTiling}
  n4 = ComponentMask(None=VectorParameter() {group=Suntide, parameterName=RuneUVOffset, defaultValue=<rgba 0 0 0 1>}.RGB) {g=true, r=true}
  n5 = TextureSample(UVs=Add(A=Multiply(A=@n2, B=@n3), B=@n4)) {samplerType=SAMPLERTYPE_LinearColor, texture=/Game/Athena/Prototype/Blueprints/Cube/Materials/DFHeightMask}
  n6 = Multiply(A=@n4) {constB=2}
  n7 = Multiply(A=@n2, B=Multiply(A=@n3, B=Constant() {r=2}))
  n8 = MaterialFunctionCall(Seed=ScalarParameter() {defaultValue=20, group=Suntide, parameterName=RuneRandomSeed}, Vec2=Ceil(None=Add(A=@n6, B=@n7))) {materialFunction=&MF_Hash}
  n9 = ComponentMask(None=@n8.Vec2) {r=true}
  LinearInterpolate(A=@n5.R, B=@n5.R, Alpha=NamedRerouteDeclaration(None=Step(Y=@n9) {constX=0.5}) {name=Random01, nodeColor=<rgba 0.5234 0 1 1>})
  n10 = NamedRerouteUsage()
  n11 = Power(Base=@n10, Exp=ScalarParameter() {defaultValue=1, group=Suntide, parameterName=RuneEmissiveExponent})
  LinearInterpolate(
   A=@n11
   B=Multiply(A=@n11, B=NamedRerouteUsage())
   Alpha=ScalarParameter() {defaultValue=1, group=Suntide, parameterName=RuneAnimationStrength}
  )
  n12 = TextureObjectParameter() {group=Suntide, parameterName=Noise, texture=/Game/Effects/FogVolume/noise_flamy_Tile}
  n13 = Normalize(VectorInput=VectorParameter() {group=Suntide, parameterName=NoiseDirection, defaultValue=<rgba 0.2 1 -3 1>}.RGB)
  n14 = Multiply(A=Time(), B=ScalarParameter() {defaultValue=3, group=Suntide, parameterName=NoiseSpeed})
  n15 = VertexInterpolator(VS=PreSkinnedPosition())
  n16 = Add(
   A=Add(A=Multiply(A=@n13, B=@n14), B=NamedRerouteDeclaration(None=@n15.PS) {name=PS_Pos, nodeColor=<rgba 1 0.5625 0 1>})
   B=NamedRerouteUsage()
  )
  n17 = NamedRerouteDeclaration(None=VertexInterpolator(VS=PreSkinnedNormal()).PS) {name=PS_Normal, nodeColor=<rgba 1 0.5625 0 1>}
  n18 = ScalarParameter() {defaultValue=75, group=Suntide, parameterName=NoiseSize}
  MaterialFunctionCall!(
   ColorTex=@n12
   NormalTex
   Pos=@n16
   Nrm=@n17
   ColorUVScale=@n18
   ColorBlendFactor=ScalarParameter() {defaultValue=8, group=Suntide, parameterName=NoiseColorBlendFactor}
   NormalStrength
   NormalUVScale
   NormalBlendFactor
   Local Space Inputs=StaticBool() {value=true}
  ) {materialFunction=/Game/Global/MaterialFunctions/MF_BiPlanar}
  n19 = TextureObjectParameter() {parameterName=warper, texture=/Game/Global/Textures/Default/Noise/T_Noise_Cloud}
  n20 = Add(
   A=NamedRerouteUsage()
   B=Multiply(A=VectorParameter() {group=Suntide, parameterName=NoiseWarpSpeed, defaultValue=<rgba 0.2 0.2 1 1>}.RGB, B=Time())
  )
  n21 = NamedRerouteUsage()
  n22 = ScalarParameter() {defaultValue=32, parameterName=NoiseWarpUVScale}
  MaterialFunctionCall(ColorTex=@n19, Pos=@n20, Nrm=@n21, ColorUVScale=@n22) {materialFunction=&MF_BiPlanar}
  Multiply(A=@n1, B=ScalarParameter() {defaultValue=1, group=Suntide, parameterName=Debug})
  n23 = TextureSampleParameter2D() {group=Suntide, parameterName=RuneMask, samplerType=SAMPLERTYPE_LinearColor, texture=/Game/Global/Textures/Default/Blanks/T_BlackLinearColor}
  NamedRerouteDeclaration(None=Power(Base=@n23.R, Exp=ScalarParameter() {defaultValue=5, group=Suntide, parameterName=RuneExponent})) {name=RuneMask, nodeColor=<rgba 0.8984 0 1 1>}
  n24 = ComponentMask(None=NamedRerouteUsage()) {b=true, g=true, r=true}
  n25 = ComponentMask(None=NamedRerouteUsage()) {r=true}
  n26 = Multiply(A=ScalarParameter() {defaultValue=0.6, group=Suntide, parameterName=RuneBaseColorMultiplier}, B=@n24)
  n27 = Add(A=@n25, B=ScalarParameter() {defaultValue=0.15, group=Suntide, parameterName=RuneRoughnessBias})
  n28 = LinearInterpolate(A=AppendVector(A=@n24, B=@n25), B=Saturate(None=AppendVector(A=@n26, B=@n27)), Alpha=NamedRerouteUsage())
  NamedRerouteDeclaration(None=SetMaterialAttributes!(
    MaterialAttributes=NamedRerouteUsage()
    BaseColor=ComponentMask(None=@n28) {b=true, g=true, r=true}
    Roughness=ComponentMask(None=@n28) {a=true}
   ) {attributeSetTypes [#1, #2]}) {name=RuneMaterial, nodeColor=<rgba 0.4531 1 0 1>}
  n29 = GetMaterialAttributes(None=NamedRerouteDeclaration(None=FunctionInput() {inputType=FunctionInput_MaterialAttributes}) {name=IN_Material, nodeColor=<rgba 0 1 0.6016 1>}) {attributeGetTypes [
    #1
    #2
    #3
    #4
   ]}
  NamedRerouteDeclaration(None=@n29.EmissiveColor) {name=IN_Emissive, nodeColor=<rgba 1 0 0.4453 1>}
  NamedRerouteDeclaration(None=@n29.Normal) {name=IN_Normal, nodeColor=<rgba 0.9219 1 0 1>}
  n30 = NamedRerouteUsage()
  n31 = Power(Base=@n30, Exp=ScalarParameter() {defaultValue=1.25, group=Suntide, parameterName=NoiseExponent})
  n32 = ScalarParameter() {defaultValue=6, group=Suntide, parameterName=NoiseStrength}
  n33 = ScalarParameter() {defaultValue=0.5, group=Suntide, parameterName=EnergyFresnelExponent}
  n34 = Multiply(
   A=ScalarParameter() {defaultValue=1, group=Suntide, parameterName=EnergyFresnelStrength}
   B=Saturate(None=Fresnel(ExponentIn=@n33) {baseReflectFraction=0})
  )
  NamedRerouteDeclaration(None=Multiply(A=Multiply(A=@n31, B=@n32), B=@n34)) {name=EnergyMask, nodeColor=<rgba 1 0.2109 0 1>}
  n35 = GetMaterialAttributes(None=MaterialFunctionCall!(Alpha=Constant() {r=1}, Material Attributes=NamedRerouteUsage()) {materialFunction=/Game/Weapons/WeaponSkins/MaterialFunction/ThinFilmFunction}.Result) {attributeGetTypes [
    #3
    #1
    #5
    #2
   ]}
  NamedRerouteDeclaration(None=@n35.Roughness) {name=MetalRoughness, nodeColor=<rgba 0.5 1 0 1>}
  n36 = NamedRerouteUsage()
  n37 = ScalarParameter() {defaultValue=500, group=Suntide, parameterName=RuneAnimOffset}
  n38 = Multiply(A=Add(A=Time(), B=Multiply(A=@n36, B=@n37)), B=ScalarParameter() {defaultValue=0.1, group=Suntide, parameterName=RuneEmissiveSpeed})
  NamedRerouteDeclaration(None=ConstantBiasScale(None=Sine(None=@n38))) {name=RuneEmissiveAnim, nodeColor=<rgba 0 1 0.4844 1>}
  n39 = Power(Base=@n30, Exp=ScalarParameter() {defaultValue=1.5, group=Suntide, parameterName=RuneEnergyMaskExponent})
  n40 = ScalarParameter() {defaultValue=4, group=Suntide, parameterName=RuneEnergyMaskMultiplier}
  NamedRerouteDeclaration(None=Multiply(A=Multiply(A=@n39, B=@n40), B=@n34)) {name=RuneEnergyMask, nodeColor=<rgba 0.2422 0 1 1>}
  n41 = ScalarParameter() {defaultValue=4, group=Suntide, parameterName=RuneEmissiveStrength}
  NamedRerouteDeclaration(None=Multiply(A=Multiply(A=@n10, B=@n41), B=NamedRerouteUsage())) {name=RuneEmissive, nodeColor=<rgba 1 0 0.8437 1>}
  NamedRerouteDeclaration(None=@n35.MaterialAttributes) {name=MetalMaterial, nodeColor=<rgba 0 0.0625 1 1>}
  NamedRerouteDeclaration(None=@n35.EmissiveColor) {name=MetalEmissive, nodeColor=<rgba 0 0.9766 1 1>}
  NamedRerouteDeclaration(None=@n35.BaseColor) {name=MetalBaseColor, nodeColor=<rgba 0.4062 0 1 1>}
  NamedRerouteDeclaration(None=@n35.Metallic) {name=MetalMetallic, nodeColor=<rgba 0 0.6016 1 1>}
  n42 = NamedRerouteUsage()
  n43 = NamedRerouteUsage()
  n44 = Multiply(
   A=Power(Base=Saturate(None=@n43), Exp=ScalarParameter() {defaultValue=2, group=Suntide, parameterName=EnergyEmissiveExp})
   B=ScalarParameter() {defaultValue=0.025, group=Suntide, parameterName=EnergyEmissiveStrength}
  )
  n45 = LinearInterpolate(
   A=NamedRerouteUsage()
   B=VectorParameter() {group=Suntide, parameterName=Energy Emissive Color, defaultValue=<rgba 0.1916 0 0.5625 1>}.RGB
   Alpha=NamedRerouteDeclaration(None=Add(A=@n42, B=@n44)) {name=EmissiveMask, nodeColor=<rgba 0 1 0.4141 1>}
  )
  n46 = Saturate(None=@n45)
  n47 = VectorParameter() {group=Suntide, parameterName=EnergyBaseColor, defaultValue=<rgba 0.1457 0.1214 0.1927 1>}
  n48 = ScalarParameter() {defaultValue=1, group=Suntide, parameterName=RuneBaseColorMultiplier}
  n49 = Saturate(None=Subtract(A=NamedRerouteUsage(), B=NamedRerouteUsage()))
  n50 = SetMaterialAttributes!(
   MaterialAttributes=NamedRerouteUsage()
   EmissiveColor=NamedRerouteDeclaration(None=MaterialFunctionCall(LDR Input=@n46, HDR Tint=@n45) {materialFunction=&DeriveHDRfromLDR}.RemappedRange) {name=SuntideEmissive, nodeColor=<rgba 0.5 1 0 1>}
   BaseColor=LinearInterpolate(A=NamedRerouteUsage(), B=Multiply(A=@n47.RGB, B=@n48), Alpha=@n49)
   Metallic=LinearInterpolate(A=NamedRerouteUsage(), B=ScalarParameter() {defaultValue=0.5, group=Suntide, parameterName=EnergyMetalness}, Alpha=@n49)
   Roughness=LinearInterpolate(A=NamedRerouteUsage(), B=ScalarParameter() {defaultValue=0.5, group=Suntide, parameterName=EnergyRoughness}, Alpha=@n49)
  ) {attributeSetTypes [
    #3
    #1
    #5
    #2
   ]}
  NamedRerouteDeclaration(None=@n50) {name=SunTideMaterial, nodeColor=<rgba 0 0.5078 1 1>}
  NamedRerouteDeclaration(None=ComponentMask(None=@n8.Vec2) {g=true}) {name=Random02, nodeColor=<rgba 0.8281 1 0 1>}
  NamedRerouteDeclaration(None=ComponentMask(None=MaterialFunctionCall(TextureObject=@n12, TextureSize=@n18, WorldPosition=@n16, World Space Normal=@n17) {materialFunction=&WorldAlignedTexture}.XYZ Texture) {r=true}) {name=Noise_01, nodeColor=<rgba 0.2656 0 1 1>}
  n51 = MaterialFunctionCall(TextureObject=@n19, TextureSize=@n22, WorldPosition=@n20, World Space Normal=@n21) {materialFunction=&WorldAlignedTexture}
  NamedRerouteDeclaration(None=Multiply(
    A=ConstantBiasScale(None=@n51.XYZ Texture) {bias=-0.5, scale=2}
    B=ScalarParameter() {defaultValue=8, group=Suntide, parameterName=NoiseWarpStrength}
   )) {name=NoiseWarp, nodeColor=<rgba 0.2891 0 1 1>}
  NamedRerouteDeclaration(None=@n29.BaseColor) {name=IN_BaseColor, nodeColor=<rgba 1 0.4219 0 1>}
  NamedRerouteDeclaration(None=@n29.Roughness) {name=IN_Roughness, nodeColor=<rgba 0 0.4375 1 1>}
  NamedRerouteUsage()
 ]}
