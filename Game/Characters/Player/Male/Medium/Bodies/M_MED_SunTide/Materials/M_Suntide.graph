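# Character body material for the SunTide player skin on skeletal meshes, using material attributes with Default Lit, Subsurface Profile and Cloth shading.
# The shared PreFX function builds the base, then MF_Suntide applies the skin effect through a SkinFX_Mask channel.
# Static switches optionally layer cloth fuzz, metal LUT and crystal functions, each masked by selectable channels of that texture.
# The result passes through shared character PostFX and cinematic PostFX functions.
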Material /Game/Characters/Player/Male/Medium/Bodies/M_MED_SunTide/Materials/M_Suntide
graph {
 properties {
  bAutomaticallySetUsageInEditor=false
  bUseMaterialAttributes=true
  bUsedWithClothing=true
  bUsedWithMeshDeformer=true
  bUsedWithMorphTargets=true
  bUsedWithNanite=true
  bUsedWithSkeletalMesh=true
  pixelDepthOffsetMode=PDOM_Legacy
  shadingModel=MSM_FromMaterialExpression
  subsurfaceProfile=&SS_HeroSkin_02
  usedShadingModels=Default Lit | Subsurface Profile | Cloth
 }
 nodes [
  n1 = MaterialFunctionCall() {materialFunction=&BaseFN_CharacterFunction_PreFX}
  n2 = TextureSample(Tex=TextureObjectParameter() {group=Skin FX, parameterName=SkinFX_Mask}) {samplerType=SAMPLERTYPE_Masks, texture=/Game/Characters/Player/Male/Medium/Bodies/M_MED_SunTide/Textures/T_M_Med_Suntide_BodyUpgrade_T2_FX}
  n3 = ChannelMaskParameter(None=@n2.RGBA) {parameterName=CrystalEdgeMapChannel}
  n4 = MaterialFunctionCall!(Alpha=ChannelMaskParameter(None=@n2.RGBA) {parameterName=SuntideChannel}.R, In=@n1.Result) {materialFunction=&/MF_Suntide}
  n5 = StaticSwitchParameter(
   True=MaterialFunctionCall!(Alpha=ChannelMaskParameter(None=@n2.RGBA) {parameterName=ClothFuzzChannel}.R, Gmap=@n1.Gmap, Material=@n4.Result) {materialFunction=/Game/Weapons/WeaponSkins/MaterialFunction/ClothFuzz}.Result
   False=@n4.Result
  ) {group=Skin FX, parameterName=Use Cloth fuzz}
  n6 = StaticSwitchParameter(
   True=MaterialFunctionCall(Mask=ChannelMaskParameter(None=@n2.RGBA) {parameterName=MetalLUTChannel}.R, Materials=@n5) {materialFunction=&MF_MetalsLUT}.Result
   False=@n5
  ) {group=Skin FX, parameterName=UseMetalLUT}
  n7 = ChannelMaskParameter(None=@n2.RGBA) {parameterName=CrystalChannel}
  n8 = MaterialFunctionCall(
   AO=@n1.AO
   Distance Blend=@n1.AO
   Fresnel=@n1.AO
   Gmap=@n1.AO
   Pre Skinned Local Position=@n1.AO
   Material=StaticSwitchParameter(
    True=MaterialFunctionCall!(Edge Map=@n3.R, Material=@n6, Alpha=@n7.R) {materialFunction=/Game/Weapons/WeaponSkins/MaterialFunction/Crystal}.Result
    False=@n6
   ) {group=Skin FX, parameterName=UseCrystal}
  ) {materialFunction=&BaseFN_CharacterFunction_PostFX}
  MaterialFunctionCall(In=@n8.Result) {materialFunction=&CINE_MF_CharPostFX}
 ]
}
